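; Runtime data Blueprint for the WallStick prototype, apparently holding per-instance state for a wall-stick movement mode.
; It keeps an ElapsedTime variable, which AddElapsedTime increments by a given delta time and returns.
; A reset-to-default-values event is present but has no logic attached.
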(event EventResettoDefaultValues)

(fn AddElapsedTime
  (params (DeltaTime :type Float%20%28double-precision%29))
  (bind _output_get (Variables|Default|SetElapsedTime (+ (Variables|Default|GetElapsedTime) DeltaTime)))
  (return _output_get))
